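; Preview actor for cosmetic companions shown on a pawn, working through the companion actor held by its locker preview component.
; When preview visuals spawn, it hides the companion mesh.
; When they are shown, a timeline fires events to show the preview, restore the mesh and play the companion's res-in effect.
; A showMesh flag guards each toggle, because hiding twice would lose the mesh's cached original position.
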(event EventOnPreviewVisualsSpawned
  (|CallParentFunction)
  (bind _asbp_fort_player_companion_preview_actor (Utilities|Casting|CastToBP_FortPlayerCompanionPreviewActor (Class|FortPlayerCompanionPreviewLockerComponent|GetCompanionActor (Variables|Default|GetBP_FortPlayerCompanionPreviewLockerComponent))))
  ; Because ShowMesh(false) moves the mesh to a set position and caches the old position, calling ShowMesh(false) twice in a row will make it impossible to ever retrieve the mesh. To prevent this, we use this bool.
  (if (Variables|Default|GetshowMesh)
    (bind _output_get (Variables|Default|SetshowMesh false))
    (Default|ShowMesh _asbp_fort_player_companion_preview_actor _output_get)))

(event EventOnPreviewVisualsShown
  (bind _bp_fortplayercompanionpreviewlockercomponent (Variables|Default|GetBP_FortPlayerCompanionPreviewLockerComponent))
  (bind _returnvalue (Class|FortPlayerCompanionPreviewLockerComponent|GetCompanionActor _bp_fortplayercompanionpreviewlockercomponent))
  (|CallParentFunction)
  (Timeline|Timeline PlayFromStart
    (:ShowPawn
      (CallFunction|ShowPreview)
      (bind _asbp_fort_player_companion_preview_actor (Utilities|Casting|CastToBP_FortPlayerCompanionPreviewActor _returnvalue))
      (if (not (Variables|Default|GetshowMesh))
        (bind _output_get (Variables|Default|SetshowMesh true))
        (Default|ShowMesh _asbp_fort_player_companion_preview_actor _output_get)))
    (:PlayResIn
      (Class|FortPlayerCompanionPreviewActor|PlayResinEffect _returnvalue))) ; Sequence Show / Outro / ResIn
)

(fn ConstructionScript
  (|CallParentFunction))
