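; Companion cosmetic preview actor that displays an off-pawn companion with its own floor and lighting.
; When shown, a timeline reveals the child companion actor, briefly toggles its mesh and plays the resin effect.
; Spawning applies the floor display setting, initializes lights and binds user-settings updates that reselect PC or mobile lighting by platform, performance mode and shadow quality.
; Custom floor materials get tuned surface parameters, with a default material as fallback.
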(event EventOnPreviewVisualsBeginLoading
  (CallFunction|HidePreview)
  (CallFunction|SetFloorEnabled (Variables|Default|GetShow%20Floor)))

(event EventOnPreviewVisualsSpawned
  (Variables|Default|SetShow%20Floor (Utilities|Struct|BreakCharacterDisplaySettings (Variables|FortItemPreviewActor|GetCurrentDisplayInfo)))
  (Utilities|FlowControl|Sequence
    (:then_0
      (CallFunction|SetFloorEnabled (Variables|Default|GetShow%20Floor)))
    (:then_1
      (CallFunction|SetUPLighting)
      (CallFunction|LightControl true))
    (:then_2
      (bind _returnvalue (Settings|GetGameUserSettings))
      (EventDispatchers|BindEventtoOnGameUserSettingsUINeedsUpdate _returnvalue (delegate Custom|UpdateSettings)))))

(event Custom|UpdateSettings
  (CallFunction|HandleLightingSettingsChanged))

(event EventOnSetFloorMaterial
  (params (InMaterialInstance :type Material%20Interface%20Object%20Reference))
  ;[0 0 0] OnSetFloorMaterial
  (bind _output_get (Variables|Default|SetCustomFloorMaterial InMaterialInstance))
  (bind _asmaterial_instance_dynamic (Utilities|Casting|CastToMaterialInstanceDynamic _output_get)
    (:then
      (Rendering|Material|SetScalarParameterValue _asmaterial_instance_dynamic "Floor Metalic" 0.0 :!class /Script/Engine.MaterialInstanceDynamic) ; MF_UI_MPItemShop_Setup
      (Rendering|Material|SetScalarParameterValue _asmaterial_instance_dynamic "Floor Roughness" 0.75 :!class /Script/Engine.MaterialInstanceDynamic) ; MF_UI_MPItemShop_Setup
      (Rendering|Material|SetScalarParameterValue _asmaterial_instance_dynamic "Floor Spec" 0.0 :!class /Script/Engine.MaterialInstanceDynamic)
      (Rendering|Material|SetScalarParameterValue _asmaterial_instance_dynamic "Floor-Metallic" 0.2 :!class /Script/Engine.MaterialInstanceDynamic) ; M_UI_FloorShadow_VaultWorld_Locker
      (Rendering|Material|SetScalarParameterValue _asmaterial_instance_dynamic "Floor-Roughness" 0.75 :!class /Script/Engine.MaterialInstanceDynamic) ; M_UI_FloorShadow_VaultWorld_Locker
    )
    (:CastFailed
      _))
  (Rendering|Material|SetMaterial (Variables|Default|GetPlane) 0 (select (Utilities|IsValid (Variables|Default|GetCustomFloorMaterial) :!class /Script/Engine.KismetSystemLibrary) (Variables|Default|GetCustomFloorMaterial) (Variables|Default|GetDefaultFloorMaterial))))

(event EventOnPreviewVisualsShown
  (Timeline|Timeline_1 PlayFromStart
    (:ShowPawn
      (CallFunction|ShowPreview)
      (bind _asbp_fort_player_companion_preview_actor (Utilities|Casting|CastToBP_FortPlayerCompanionPreviewActor (Variables|/Script/Engine.ChildActorComponent|GetChildActor (Variables|FortItemPreviewOffPawnActor|GetPreviewActorComponent))))
      (Default|ShowMesh _asbp_fort_player_companion_preview_actor false)
      (Utilities|FlowControl|Delay 0.2)
      (Default|ShowMesh _asbp_fort_player_companion_preview_actor true))
    (:PlayResIn
      (bind _asfort_player_companion_preview_actor (Utilities|Casting|CastToFortPlayerCompanionPreviewActor (Variables|/Script/Engine.ChildActorComponent|GetChildActor (Variables|FortItemPreviewOffPawnActor|GetPreviewActorComponent))))
      (Class|FortPlayerCompanionPreviewActor|PlayResinEffect _asfort_player_companion_preview_actor))) ; Sequence Show / Outro / ResIn
)

(fn HandleLightingSettingsChanged
  (bind _returnvalue (TouchPlatform|IsMobileGame))
  (bind _returnvalue_2 (Settings|GetFortGameUserSettings))
  (bind _returnvalue_3 (Class|FortGameUserSettings|IsinPerformanceMode _returnvalue_2))
  (bind _returnvalue_4 (Math|Boolean|ORBoolean _returnvalue _returnvalue_3 false))
  (bind _returnvalue_5 (Settings|GetGameUserSettings))
  (if (Variables|Default|GetIs%20%20active)
    (if (or (== (Settings|GetShadowQuality _returnvalue_5) 0) _returnvalue_4)
      (if _returnvalue_4
        (CallFunction|SwitchPCLighting false)
        (CallFunction|SwitchMobileLighting true)
        (else
          (CallFunction|SwitchPCLighting true)))
      (else
        (CallFunction|SwitchMobileLighting false)
        (CallFunction|SwitchPCLighting true)))))

(fn LightControl
  (params (Active :type Boolean))
  (bind _returnvalue (TouchPlatform|IsMobileGame))
  (bind _returnvalue_2 (TouchPlatform|IsNintendoSwitch))
  (bind _returnvalue_3 (TouchPlatform|IsNintendoSwitch))
  (bind _returnvalue_4 (Settings|GetFortGameUserSettings))
  (bind _returnvalue_5 (Class|FortGameUserSettings|IsinPerformanceMode _returnvalue_4))
  (bind _returnvalue_6 (Math|Boolean|ORBoolean _returnvalue _returnvalue_5 _returnvalue_2))
  (bind _is__active (Variables|Default|GetIs%20%20active))
  (Variables|Default|SetIs%20%20active Active)
  (bind _returnvalue_7 (Settings|GetGameUserSettings))
  (if (Math|Boolean|ORBoolean (== (Settings|GetShadowQuality _returnvalue_7) 0) _returnvalue_6 _returnvalue_3)
    (if _returnvalue_6
      (CallFunction|SwitchMobileLighting _is__active)
      (elif _returnvalue_3
        (CallFunction|SwitchErebusLighting _returnvalue_3))))
  (CallFunction|SwitchPCLighting _is__active))

(fn SetUPLighting
  (Rendering|Components|Light|SetForceCachedShadowsforMovablePrimitives (Variables|Default|GetRimTop_pc) true)
  (CallFunction|SwitchMobileLighting false)
  (CallFunction|SwitchPCLighting false))

(fn SetFloorEnabled
  (params (Show%20Floor :type Boolean))
  (Development|SetHiddeninGame (Variables|Default|GetPlane) (not Show%20Floor) true))

(fn SwitchErebusLighting
  (params (visibility :type Boolean))
  (Rendering|SetVisibility (Variables|Default|GetRimTop_pc) false false)
  (Rendering|SetVisibility (Variables|Default|GetDirectionalLight_mobile) visibility false))

(fn SwitchMobileLighting
  (params (NewParam :type Boolean))
  (Rendering|SetVisibility (Variables|Default|GetRimTop_pc) false false)
  (Rendering|SetVisibility (Variables|Default|GetDirectionalLight_mobile) NewParam false))

(fn SwitchPCLighting
  (params (Visibility :type Boolean))
  (Rendering|SetVisibility (Variables|Default|GetRimTop_pc) Visibility false)
  (Rendering|SetVisibility (Variables|Default|GetDirectionalLight_mobile) false false))

(fn ConstructionScript)
